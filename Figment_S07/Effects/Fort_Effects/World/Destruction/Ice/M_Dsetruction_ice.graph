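# Masked Default Lit material for ice destruction effects, usable on sprite and mesh particles.
# Color blends two blues by a reused corrupt rock texture, then adds highlights from the powered, summed absolute components of that rock's normal map.
# MF_Dissolve erodes a tiling smoke texture using inverted particle alpha.
# Particle red and green scale the erosion UVs, and green also scales a tile normal sample's UVs.
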Material /Figment_S07/Effects/Fort_Effects/World/Destruction/Ice/M_Dsetruction_ice
graph {
 properties {
  bUsedWithMeshParticles=true
  bUsedWithParticleSprites=true
  blendMode=BLEND_Masked
  pixelDepthOffsetMode=PDOM_Legacy
  usedShadingModels=Default Lit
 }
 nodes [
  n1 = Constant3Vector() {constant=<rgba 0 0 1 0>}
  n2 = TextureCoordinate() {uTiling=0.55, vTiling=0.5}
  n3 = TextureSample(UVs=@n2) {samplerType=SAMPLERTYPE_Normal, texture=/Game/Environments/Props/Rocks/Textures/T_Corrupt_Rock_AB_N}
  n4 = Ceil(None=ComponentMask(None=TextureCoordinate()) {g=true})
  n5 = MaterialFunctionCall(f2=Abs(None=LinearInterpolate(A=@n1, B=@n3.RGB, Alpha=@n4))) {materialFunction=&AddComponents}
  n6 = Constant() {r=6}
  Add(
   A=Multiply(A=Power(Base=@n5.f2, Exp=@n6)) {constB=4}
   B=LinearInterpolate(
    A=VectorParameter() {parameterName=Param_1, defaultValue=<rgba 0.2627 0.6886 0.98 1>}.RGB
    B=VectorParameter() {defaultValue=<rgba 0.0482 0.2716 0.9688 1>}.RGB
    Alpha=TextureSample(UVs=@n2) {texture=/Game/Environments/Props/Rocks/Textures/T_Corrupt_Rock_AB_D}.G
   )
  )
  n7 = ParticleColor()
  n8 = Multiply(A=TextureCoordinate(), B=@n7.G)
  MaterialFunctionCall!(
   Ramp=TextureSampleParameter2D(UVs=Multiply(A=@n8, B=@n7.R)) {parameterName=ErodeTexture, sortPriority=3, texture=/Game/Effects/Fort_Effects/Textures/Tile/T_Inky_Smoke_Tile}.R
   Dissolve=OneMinus(None=@n7.A)
   EdgeSharpness
  ) {materialFunction=/Game/Effects/Fort_Effects/Materials/_Functions/MF_Dissolve}
  Multiply(A=ComponentMask(None=@n2) {r=true}, B=ComponentMask(None=@n2) {g=true})
  ScalarParameter() {defaultValue=2, group=Tile, parameterName=TileTextureScale}
  n9 = TextureObjectParameter() {group=Tile, parameterName=N_Tile, samplerType=SAMPLERTYPE_Normal, texture=&T_Corrupt_Rock_AB_N}
  TextureSample(UVs=Multiply(A=TextureCoordinate(), B=@n7.G), Tex=@n9) {samplerType=SAMPLERTYPE_Normal, texture=/Figment_S07/Athena/WinterArt/Textures/T_IceScratch1_N}
 ]
}
